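; Day sequence actor for the Sparks Festival Season 8 main stage lobby, managing baked cloud meshes and low-end exposure.
; Construction runs BakedCloudsVisibility, which drops the cloud root far below the world when ShowBakedClouds is off and fills the hidden and all-mesh cloud lists.
; SetLowEndExposure, not called within this graph, enables a fallback exposure post process when Lumen is off and logs the result.
; BeginPlay and Tick do nothing.
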(event EventBeginPlay
  (|CallParentFunction) ; This node is disabled and will not be called.\nDrag off pins to build functionality.
)

(event EventTick
  (params (DeltaSeconds :type Float%20%28single-precision%29)))

(fn BakedCloudsVisibility
  (Transformation|SetWorldLocation (Variables|Default|GetBakedClouds) (select (Variables|Default|GetShowBakedClouds) "0, 0, 0" "0.000000,0.000000,-10000000.000000") false false)
  (Variables|Clouds|SetHiddenMeshClouds (Utilities|Array|MakeArray (Variables|Default|GetBakedClouds_HighA) (Variables|Default|GetBakedClouds_HighA2) (Variables|Default|GetBakedClouds_MidA3) (Variables|Default|GetBakedClouds_HighB) (Variables|Default|GetBakedClouds_HighC1) (Variables|Default|GetBakedClouds_HighEd)))
  (Variables|Clouds|SetAllMeshClouds (Components|GetChildrenComponents (Variables|Default|GetBakedClouds) false)))

(fn SetLowEndExposure
  (bind _isenabled (Class|BTechArtGlobalLibrary|IsLumenEnabled))
  (bind _returnvalue (not _isenabled))
  (bind _exposurelowendpostprocess (Variables|Default|GetExposureLowEndPostProcess))
  (Rendering|SetVisibility _exposurelowendpostprocess _returnvalue false)
  (Development|SetHiddeninGame _exposurelowendpostprocess _isenabled false)
  (Fortnite|FortShippingLog self (Utilities|String|Append "TODM Scalability: Low End Exposure = " (Utilities|String|ToString(Boolean) _returnvalue))))

(fn ConstructionScript
  (|CallParentFunction)
  (CallFunction|BakedCloudsVisibility))
