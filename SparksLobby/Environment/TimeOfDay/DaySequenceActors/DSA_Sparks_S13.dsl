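; DSA_Sparks_S13 is a day sequence actor Blueprint for the Season 13 Sparks lobby time-of-day environment, with a toggle for baked clouds.
; Its construction script calls the parent, then BakedCloudsVisibility, which reads ShowBakedClouds and sets the BakedClouds component's world location.
; Instead of changing visibility, it hides the clouds by moving them far below the world, placing them at the origin when shown.
; BeginPlay and Tick contain no active logic.
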(event EventBeginPlay
  (|CallParentFunction) ; This node is disabled and will not be called.\nDrag off pins to build functionality.
)

(event EventTick
  (params (DeltaSeconds :type Float%20%28single-precision%29)))

(fn BakedCloudsVisibility
  (Transformation|SetWorldLocation (Variables|Default|GetBakedClouds) (select (Variables|Default|GetShowBakedClouds) "0, 0, 0" "0.000000,0.000000,-10000000.000000") false false))

(fn ConstructionScript
  (|CallParentFunction)
  (CallFunction|BakedCloudsVisibility))
